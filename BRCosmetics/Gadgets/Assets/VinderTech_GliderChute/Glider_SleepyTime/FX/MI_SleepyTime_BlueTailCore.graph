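# Translucent, unlit, two-sided material for the blue core of the SleepyTime glider's tail effect.
# A powered V-coordinate gradient, scaled by brightness, is added to the same gradient multiplied by panning streak-texture noise.
# The sum blends between two blue colors, the second in HDR range, and feeds MF_FX_EmissiveAdjustments.
# SpeedX and SpeedY only offset the panned UVs, as the Panner's speeds are fixed.
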Material /BRCosmetics/Gadgets/Assets/VinderTech_GliderChute/Glider_SleepyTime/FX/MI_SleepyTime_BlueTailCore
graph {
 properties {blendMode=BLEND_Translucent, pixelDepthOffsetMode=PDOM_Legacy, shadingModel=MSM_Unlit, twoSided=true, usedShadingModels=Unlit}
 nodes [
  n1 = Power(Base=ComponentMask(None=TextureCoordinate()) {g=true}, Exp=ScalarParameter() {defaultValue=3, parameterName=Exponent})
  n2 = Multiply(A=@n1, B=ScalarParameter() {defaultValue=3, parameterName=Brightness})
  n3 = Add(
   A=Panner(Coordinate=TextureCoordinate() {vTiling=0.2}) {bFractionalPart=true, speedX=0.2, speedY=0.3}
   B=AppendVector(A=ScalarParameter() {parameterName=SpeedX}, B=ScalarParameter() {parameterName=SpeedY})
  )
  n4 = Multiply(
   A=@n1
   B=Multiply(
    A=TextureSample(UVs=@n3) {samplerType=SAMPLERTYPE_LinearColor, texture=/Game/Effects/Fort_Effects/Textures/Tile/T_Raygun_Muzzle_02}.R
    B=ScalarParameter() {defaultValue=1, parameterName=StreakBrightness}
   )
  )
  MaterialFunctionCall(Emissive=LinearInterpolate(
    A=VectorParameter() {parameterName=Color A, defaultValue=<rgba 0 0 1 1>}.RGB
    B=VectorParameter() {parameterName=Color B, defaultValue=<rgba 0 54.1664 250 1>}.RGB
    Alpha=Add(A=@n2, B=@n4)
   )) {materialFunction=&MF_FX_EmissiveAdjustments}
 ]
}
