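# Masked two-sided material for the tail of the SleepyTime glider, showing glowing drifting stars over an animated blue gradient.
# Three panning star samples and blurred noise build the Fresnel-scaled emissive color, with flat white on back faces.
# A sine/cosine offset ripples along UV V, while SpeedUp shifts the wave and scrolling textures.
# Panning rift noise, column-quantized streak noise and curve-atlas gradients form an erosion mask blended by the Appearing parameter.
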Material /BRCosmetics/Gadgets/Assets/VinderTech_GliderChute/Glider_SleepyTime/FX/M_SleepyTime_Gliader_Tail
graph {
 properties {blendMode=BLEND_Masked, pixelDepthOffsetMode=PDOM_Legacy, twoSided=true, usedShadingModels=Default Lit}
 nodes [
  Constant() {r=1}
  n1 = ScalarParameter() {defaultValue=1, parameterName=SpeedUp}
  n2 = AppendVector(A=Constant(), B=@n1)
  n3 = TextureSample!(UVs=Add(A=Panner() {speedX=-0.02, speedY=-0.2}, B=@n2), Tex, Level, Apply View MipBias) {
   constMipValue=0
   mipValueMode=TMVM_MipLevel
   texture=&/T_SleepyTime_Stars
  }
  n4 = Panner() {speedX=0.01, speedY=-0.15}
  n5 = Constant2Vector() {g=1, r=1}
  n6 = TextureSample(UVs=Add(A=@n2, B=Multiply(A=@n4, B=@n5))) {
   constMipValue=0
   mipValueMode=TMVM_MipLevel
   texture=&T_SleepyTime_Stars
  }
  n7 = Add(A=@n3.B, B=Multiply(A=@n6.R, B=ScalarParameter() {defaultValue=10, parameterName=BigStarGlow}))
  n8 = Panner() {speedX=0.015, speedY=-0.17}
  n9 = Constant2Vector() {g=2, r=2}
  n10 = TextureSample(UVs=Add(A=@n1, B=Multiply(A=@n8, B=@n9))) {
   constMipValue=0
   mipValueMode=TMVM_MipLevel
   texture=&T_SleepyTime_Stars
  }
  n11 = VectorParameter() {parameterName=MixMask2, defaultValue=<rgba 4 0.2 -0.01 -0.3>}
  n12 = AppendVector(A=@n11.B, B=@n11.B)
  n13 = Add(
   A=AppendVector(A=Constant(), B=ScalarParameter() {defaultValue=1, parameterName=SpeedUp})
   B=Multiply(A=Panner(Speed=@n12), B=AppendVector(A=@n11.R, B=@n11.R))
  )
  n14 = CurveAtlasRowParameter(CurveTime=ComponentMask(None=TextureCoordinate()) {g=true}) {
   atlas=&AthenaCurveAtlas01
   curve=/Game/Gadgets/Assets/VinderTech_GliderChute/Glider_SleepyTime/FX/Curve_SleepyTime_Glider2
   defaultValue=129
  }
  n15 = Multiply(
   A=Multiply(A=TextureSample!(UVs=@n13, Tex, Apply View MipBias) {samplerType=SAMPLERTYPE_Grayscale, texture=&LowResBlurredNoise}.R) {constB=1.5}
   B=@n14.A
  ) {constB=0.1}
  n16 = Power(Base=ComponentMask(None=TextureCoordinate()) {g=true})
  n17 = Sine(None=Multiply(A=OneMinus(None=@n16)) {constB=0.25})
  n18 = Add(
   A=LinearInterpolate(B=Constant3Vector() {constant=<rgba 0 0.7156 1 0>}, Alpha=Add(A=@n7, B=@n10.R))
   B=LinearInterpolate(A=Constant3Vector() {constant=<rgba 0 1.4312 2 0>}, B=Constant3Vector() {constant=<rgba 0 0.052 1 0>}, Alpha=Add(A=@n15, B=@n17))
  )
  n19 = LinearInterpolate(Alpha=Power(Base=Fresnel() {exponent=2})) {constA=0.5, constB=4}
  n20 = TwoSidedSign()
  MaterialFunctionCall(Emissive=LinearInterpolate(A=Constant3Vector() {constant=<rgba 2 2 2 0>}, B=Multiply(A=@n18, B=@n19), Alpha=Clamp(None=@n20))) {materialFunction=&MF_FX_EmissiveAdjustments}
  n21 = Time()
  n22 = ScalarParameter() {defaultValue=1, parameterName=SpeedUp}
  n23 = Add(A=ComponentMask(None=TextureCoordinate()) {g=true}, B=Add(A=Multiply(A=@n21) {constB=-0.3}, B=Multiply(A=@n22) {constB=3})) {constB=0}
  n24 = AppendVector(A=AppendVector(A=Sine(None=@n23) {period=0.5}, B=Constant()), B=Cosine(None=@n23) {period=0.5})
  n25 = ComponentMask(None=TextureCoordinate()) {g=true}
  Multiply(A=Transform(None=Multiply(A=@n24) {constB=3}) {transformSourceType=TRANSFORMSOURCE_Local}, B=Clamp(None=Subtract(A=@n25) {constB=0.3}))
  n26 = TextureCoordinate()
  n27 = Sine(None=LinearInterpolate(Alpha=ComponentMask(None=@n26) {g=true}) {constA=0.25, constB=0.75})
  n28 = ScalarParameter() {defaultValue=1, parameterName=Appearing}
  n29 = Add(A=Divide(A=Add(A=@n27)), B=Subtract(A=@n28) {constB=0.2})
  n30 = Constant() {r=0.5}
  n31 = Add(A=@n30)
  n32 = VectorParameter() {parameterName=TopRift1, defaultValue=<rgba 0.5 0.2 0.05 -0.3>}
  n33 = AppendVector(A=@n32.B, B=@n32.B)
  n34 = TextureSample(UVs=Multiply(A=Panner(Speed=@n33), B=AppendVector(A=@n32.R, B=@n32.R))) {samplerType=SAMPLERTYPE_Normal, texture=/Game/Global/Textures/Default/Tileable_Patterns/2ndInfectedMapNormal}
  n35 = VectorParameter() {parameterName=TopRift2, defaultValue=<rgba 1 0.2 0.1 -0.1>}
  n36 = AppendVector(A=@n35.B, B=@n35.B)
  n37 = TextureSample(UVs=Multiply(A=Panner(Speed=@n36), B=AppendVector(A=@n35.R, B=@n35.R))) {samplerType=SAMPLERTYPE_Masks, texture=/Game/Global/Textures/Default/Noise/T_CellularNoise}
  n38 = CurveAtlasRowParameter(CurveTime=ComponentMask(None=TextureCoordinate()) {g=true}) {
   atlas=&AthenaCurveAtlas01
   curve=&Curve_SleepyTime_Glider2
   defaultValue=129
  }
  n39 = Add(A=Multiply(A=LinearInterpolate(A=@n30, B=@n31, Alpha=@n34.R), B=LinearInterpolate(Alpha=@n37.R)), B=@n38.R)
  n40 = ComponentMask(None=TextureCoordinate()) {g=true}
  n41 = Add(A=Power(Base=OneMinus(None=@n40))) {constB=0.5}
  n42 = AppendVector(A=Constant(), B=Multiply(A=ScalarParameter() {defaultValue=1, parameterName=SpeedUp}) {constB=0.6})
  n43 = Panner(Coordinate=TextureCoordinate()) {speedX=-0.02}
  n44 = Constant() {r=30}
  n45 = Floor(None=Multiply(A=ComponentMask(None=@n43) {r=true}, B=@n44) {constB=10})
  n46 = Time()
  n47 = Add(A=@n42, B=AppendVector(A=Divide(A=@n45, B=@n44), B=Multiply(A=@n46) {constB=0.06}))
  n48 = Panner(Coordinate=TextureCoordinate()) {speedX=0.05}
  n49 = Constant() {r=30}
  n50 = Floor(None=Multiply(A=ComponentMask(None=@n48) {r=true}, B=@n49) {constB=10})
  n51 = Time()
  n52 = Add(A=@n42, B=AppendVector(A=Divide(A=@n50, B=@n49), B=Multiply(A=@n51) {constB=0.1}))
  n53 = Multiply(A=Add(
    A=TextureSample(UVs=@n47) {texture=/Game/Global/Textures/Default/Noise/T_FireNoise_01}.R
    B=TextureSample(UVs=@n52) {samplerType=SAMPLERTYPE_Masks, texture=/Game/Weapons/WeaponSkins/Wraps/Textures/Mash/Mash_Noise}.R
   )) {constB=0.5}
  n54 = CurveAtlasRowParameter(CurveTime=ComponentMask(None=TextureCoordinate()) {g=true}) {
   atlas=&AthenaCurveAtlas01
   curve=&Curve_SleepyTime_Glider2
   defaultValue=129
  }
  n55 = If(A=ComponentMask(None=TextureCoordinate()) {g=true}, B=Constant() {r=0.5}, "A > B"=Multiply(A=Add(A=@n41, B=@n53), B=@n54.B), "A < B"=@n39)
  n56 = CurveAtlasRowParameter(CurveTime=ComponentMask(None=TextureCoordinate()) {g=true}) {
   atlas=&AthenaCurveAtlas01
   curve=&Curve_SleepyTime_Glider2
   defaultValue=129
  }
  Reroute(None=LinearInterpolate(A=Multiply(A=@n29, B=@n39), B=Multiply(A=@n55, B=@n56.G), Alpha=Clamp(None=@n28)))
 ]
}
